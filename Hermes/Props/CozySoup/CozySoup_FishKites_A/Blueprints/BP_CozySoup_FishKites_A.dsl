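; Fish kite prop that populates its static mesh's sockets with randomly chosen, tinted fish meshes and destroys itself when damaged.
; Construction shuffles sockets by random stream, optionally seeded from world position, and fills a set percentage.
; Fish get distance-limited World Position Offset and per-instance random custom primitive data, reapplied at BeginPlay, plus optional collision removal.
; RunRandomization rebuilds them from a new seed; exposed-asset overrides register the fish meshes as allowed references.
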(event EventBeginPlay
  (for (_array_element _array_index) (Variables|Default|GetAR_AddedComponents)
    (Rendering|Material|SetCustomPrimitiveDataFloat _array_element 14 (Math|Random|RandomFloatfromStream (Math|Random|MakeRandomStream (+ (* _array_index (Math|Random|BreakRandomStream (Variables|Default|GetRandomStream))) (Math|Random|BreakRandomStream (Variables|Default|GetRandomStream))))))))

(event Collision|EventActorBeginOverlap
  (params (OtherActor :type Actor%20Object%20Reference)))

(event EventTick
  (params (DeltaSeconds :type Float%20%28single-precision%29)))

(event EventOnDamageServer
  (params (Damage :type Float%20%28single-precision%29) (DamageTags :type Gameplay%20Tag%20Container%20Structure%20%28by%20ref%29) (Momentum :type Vector :default "0, 0, 0") (HitInfo :type Hit%20Result%20Structure%20%28by%20ref%29) (InstigatedBy :type Controller%20Object%20Reference) (DamageCauser :type Actor%20Object%20Reference) (EffectContext :type Gameplay%20Effect%20Context%20Handle%20Structure))
  (Actor|DestroyActor))

(fn BuildExposedAssetProviderVersion(Implementation)
  (params (Version :type Fort%20Exposed%20Asset%20Provider%20Version%20Structure%20%28by%20ref%29))
  (|CallParentFunction Version)
  (ExposedAssetProvider|AddVersion Version 1)
  (return))

(fn CollectAssetReferencestoAllow(Implementation)
  (params (Collector :type Fort%20Exposed%20Asset%20Reference%20Collector%20Structure%20%28by%20ref%29))
  (|CallParentFunction Collector)
  (for _array_element (Variables|Default|GetAR_FishMeshes)
    (ExposedAssetProvider|CollectIndirectAssetReference Collector (Utilities|ResolveSoftReference _array_element)))
  (return))

(fn EnableWPO
  (for (_array_element _array_index) (Variables|Default|GetAR_AddedComponents)
    ;[1 1 1] Material WPO
    (Rendering|LOD|SetEvaluateWorldPositionOffset _array_element true :!class /Script/Engine.StaticMeshComponent)
    (Rendering|LOD|SetWorldPositionOffsetDisableDistance _array_element 10000 :!class /Script/Engine.StaticMeshComponent)
    (Rendering|Material|SetCustomPrimitiveDataFloat _array_element 14 (Math|Random|RandomFloatfromStream (Math|Random|MakeRandomStream (+ (* _array_index (Math|Random|BreakRandomStream (Variables|Default|GetRandomStream))) (Math|Random|BreakRandomStream (Variables|Default|GetRandomStream))))))))

(fn RunRandomization
  (Variables|Default|SetRandomizeByWorldPosition false)
  (for _array_element (Variables|Default|GetAR_AddedComponents)
    (Components|DestroyComponent _array_element))
  (Utilities|Array|Clear (Variables|Default|GetAR_ValidSockets))
  (bind _returnvalue (Math|Random|MakeRandomStream (Math|Random|RandomIntegerinRange 0 9999)))
  (Variables|Default|SetRandomStream _returnvalue)
  (Default|SpawnFish)
  (CallFunction|TintFish)
  (CallFunction|EnableWPO)
  (if (Variables|Default|GetRemoveCollision)
    (for _array_element_2 (Variables|Default|GetAR_AddedComponents)
      (Collision|SetCollisionEnabled _array_element_2 "NoCollision"))))

(fn SetRandomStream
  (bind _returnvalue (Math|Random|MakeRandomStream (Math|Float|Truncate (* (Class|EnvironmentArtFuncLib|RandomFloatbyLocation (Transformation|GetActorLocation) (Variables|Default|GetRandomizeDistance) true) 1e+04))))
  (Variables|Default|SetRandomStream _returnvalue))

(fn SetSpawnPercentage
  (bind _ar_allsockets (Variables|Default|GetAR_AllSockets))
  (Utilities|Array|Clear (Variables|Default|GetAR_ValidSockets))
  (Variables|Default|SetNumberToSpawn (/ (* (Variables|Default|GetMaxSocketAmount) (Variables|Default|GetSpawnPercentage)) 100))
  (Utilities|Array|ShufflefromStream _ar_allsockets (Variables|Default|GetRandomStream))
  (for _index (range (Variables|Default|GetNumberToSpawn))
    (Utilities|Array|Add (Variables|Default|GetAR_ValidSockets) (Utilities|Array|Get(acopy) _ar_allsockets _index))))

(fn SpawnFish
  (bind _output_get (Variables|Default|SetAR_AllSockets (Transformation|GetAllSocketNames (Variables|Debug|GetStaticMeshComponent))))
  (Variables|Default|SetMaxSocketAmount (Utilities|Array|Length _output_get))
  (CallFunction|SetSpawnPercentage)
  (for _array_element (Variables|Default|GetAR_ValidSockets)
    (bind _returnvalue (AddComponent|NODE_AddStaticMeshComponent-0 :RelativeTransform (Transformation|GetSocketTransform (Variables|Debug|GetStaticMeshComponent) _array_element "RTS_Actor")))
    (Utilities|Array|Add (Variables|Default|GetAR_AddedComponents) _returnvalue)
    (Components|StaticMesh|SetStaticMesh _returnvalue (Utilities|Array|RandomArrayItemfromStream (Variables|Default|GetAR_FishMeshes) (Variables|Default|GetRandomStream)))))

(fn TintFish
  (for _array_element (Variables|Default|GetAR_AddedComponents)
    (Default|SetColoronSecondaryMesh (Utilities|Array|RandomArrayItemfromStream (Variables|Default|GetFishColors) (Variables|Default|GetRandomStream)) _array_element)))

(fn ConstructionScript
  (|CallParentFunction)
  (if (Variables|Default|GetRandomizeByWorldPosition)
    (CallFunction|SetRandomStream))
  (Default|SpawnFish)
  (CallFunction|TintFish)
  (CallFunction|EnableWPO)
  (if (Variables|Default|GetRemoveCollision)
    (for _array_element (Variables|Default|GetAR_AddedComponents)
      (Collision|SetCollisionEnabled _array_element "NoCollision"))))
